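; Parent actor for harvestable bush herbs that players interact with or destroy to get loot-table items.
; Interaction grants loot directly when the pawn's inventory item count allows, otherwise loot spawns as a world pickup, as on death.
; Clearing bLootAvailable drives its OnRep, which disables collision, interaction and marking, then kills the actor on authority.
; Death either hides visuals or, after a direct grant, timeline-shrinks the actor and its glow sprites via dynamic materials.
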(event Interaction|EventBlueprintOnInteract
  (params (InteractingPawn :type Fort%20Pawn%20Object%20Reference) (InteractionBeingAttempted :type EInteractionBeingAttempted%20Enum :default "FirstInteraction"))
  ;[0.401 0.2435 0.1681] BlueprintOnInteract
  ; Spawn Loot, either give directly to the player or drop it in the environment
  (Utilities|IsValid InteractingPawn
    (:Is%20Valid
      (bind _output_get (Variables|Default|SetLastInteractingPawn InteractingPawn))
      (if (CallFunction|PawnHasOpenInventorySlot _output_get)
        (CallFunction|GiveLoottoInteractingPawn)
        (else
          _)))
    (:Is%20Not%20Valid
      _))
  (CallFunction|SpawnLootinWorld))

(event Custom|Play%20Loot%20Pickup%20Sound
  (params (InteractingPawn :type Fort%20Pawn%20Object%20Reference))
  ; Loot Pickup sound (plays only for local client that gets the loot)
  (Networking|CanExecuteCosmeticEvents
    (:True
      (if (Pawn|IsLocallyViewed InteractingPawn)
        (Audio|PlaySound2D (Variables|Effects|GetLoot%20Pickup%20Sound))))))

(event Custom|GiveLootToInteractingPawn
  ;[0.002001 0.08149 1] GiveLootToInteractingPawn
  ; Make sure the interacting pawn is still valid and then give the loot
  (if (Variables|Default|GetbLootAvailable)
    (bind _lastinteractingpawn (Variables|Default|GetLastInteractingPawn)
      (:then
        (bind _asfort_player_controller (Utilities|Casting|CastToFortPlayerController (Pawn|GetController _lastinteractingpawn))
          (:then
            (CallFunction|GiveGrantedLoot (Variables|Default|GetLootTierGroupName) _asfort_player_controller)
            (Variables|Default|SetbLootAvailable false) ; has been looted
            (Variables|Default|SetbShrinkBeforeDeath true)
            (CallFunction|PlayLootPickupSound (Variables|Default|GetLastInteractingPawn)))
          (:CastFailed
            _)))
      (:else
        _))
    (CallFunction|SpawnLootinWorld) ; can't give loot to a player that doesnt exist, spawn it in the world
  ))

(event Custom|SpawnLootInWorld
  ;[0.002001 0.08149 1] SpawnLootInWorld
  ; Unable to give loot directly to the player, so spawn it into the world as a pickup
  (if (Variables|Default|GetbLootAvailable)
    (Class|AthenaFunctionLibrary|SpawnLootfromTable (Variables|Default|GetLootTierGroupName) 1 -1 (+ (Transformation|GetActorLocation) "0.000000,0.000000,32.000000") "0, 0, 0" 0 true true false 0 "Other" "ItemSpawner")
    (Variables|Default|SetbLootAvailable false) ; has been looted
  ))

(event Custom|RemoveAndShrink
  ;[0.4216 0.06739 0.6901] RemoveAndShrink
  (CallFunction|PrepDynamicMaterialInstance) ; Prep for glow sprite shrink
  (bind (_direction _scale) (Timeline|ShrinkToZero Play)
    (:Update
      (Transformation|SetActorRelativeScale3D _scale)
      ;;[1 1 1] Scale SpriteScale from default to 0 (based on timeline)
      (for (_array_element _array_index) (Variables|Default|GetDMI_VisComp_MatInstance)
        (Rendering|Material|SetScalarParameterValue _array_element "SpriteScale" (* (Utilities|Array|Get(aref) (Variables|Default|GetVisComp_DefaultSpriteScale) _array_index) (.x _scale)) :!class /Script/Engine.MaterialInstanceDynamic)))))

(event EventOnDeathServer
  (params (Damage :type Float%20%28single-precision%29) (DamageTags :type Gameplay%20Tag%20Container%20Structure%20%28by%20ref%29) (Momentum :type Vector :default "0, 0, 0") (HitInfo :type Hit%20Result%20Structure%20%28by%20ref%29) (InstigatedBy :type Controller%20Object%20Reference) (DamageCauser :type Actor%20Object%20Reference) (EffectContext :type Gameplay%20Effect%20Context%20Handle%20Structure))
  ;[1 0.384 0.008977 0.5] Death Events
  ;
  (CallFunction|SpawnLootinWorld) ; when destroyed, spawn loot like any other container
)

(event EventOnDeathPlayEffects
  (params (Damage :type Float%20%28single-precision%29) (DamageTags :type Gameplay%20Tag%20Container%20Structure%20%28by%20ref%29) (Momentum :type Vector :default "0, 0, 0") (HitInfo :type Hit%20Result%20Structure%20%28by%20ref%29) (InstigatedBy :type Fort%20Pawn%20Object%20Reference) (DamageCauser :type Actor%20Object%20Reference) (EffectContext :type Gameplay%20Effect%20Context%20Handle%20Structure))
  ;[1 0.384 0.008977 0.5] Death Events
  ;
  (if (Variables|Default|GetbShrinkBeforeDeath)
    (CallFunction|RemoveandShrink)
    (else
      (CallFunction|HideVisibleComponents) ; turn off visuals
    )))

(event OnLinkedActorDestroyed(FortLinkToActor)
  (params (DamageCauser :type Actor%20Object%20Reference))
  ;[1 0.384 0.008977 0.5] Death Events
  ;
  (Actor|DestroyActor) ; linked actor should be whatever this is \nsitting on (likely the basket)
)

(fn BlueprintCanInteract
  (params (InteractingPawn :type Fort%20Pawn%20Object%20Reference) (InteractionBeingAttempted :type EInteractionBeingAttempted%20Enum :default "FirstInteraction") (InteractionType :type TInteractionType%20Enum :default "IT_NoInteraction"))
  (return (Variables|Default|GetbLootAvailable)))

(fn BlueprintGetInteractionString
  (params (InteractingPawn :type Fort%20Pawn%20Object%20Reference) (InteractionBeingAttempted :type EInteractionBeingAttempted%20Enum :default "FirstInteraction"))
  (return (Variables|Interaction|GetInteractMessage)))

(fn BlueprintGetInteractionTime
  (params (InteractingPawn :type Fort%20Pawn%20Object%20Reference) (InteractionBeingAttempted :type EInteractionBeingAttempted%20Enum :default "FirstInteraction"))
  (return (ScalableFloat|GetValueatLevel (Variables|Interaction|GetRow_InteractionTime)) false))

(fn CacheVisualComponents)

(fn GiveGrantedLoot
  (params (LootTierToGrant :type Name) (PawnToGrantLoot :type Fort%20Player%20Controller%20Object%20Reference))
  (bind (_outloottodrop _returnvalue) (Loot|PickLootDrops LootTierToGrant 0 -1 "(GameplayTags=)"))
  (if _returnvalue
    (for _array_element _outloottodrop
      (bind (_array_element_itemdefinition _array_element_itemguid _array_element_count) (Class|FortItemFunctionLibrary|BreakFortItemEntry _array_element))
      (bind _asfort_world_item_definition (Utilities|Casting|CastToFortWorldItemDefinition _array_element_itemdefinition))
      (Inventory|GiveItemtoInventoryOwner :InventoryOwner PawnToGrantLoot :ItemDefinition _asfort_world_item_definition :NumberToGive _array_element_count :bNotifyPlayer true :ItemLevel -1 :PickupInstigatorHandle -1 :bUseItemPickupAnalyticEvent false :WeaponAmmoOverride -1))))

(fn HideVisibleComponents
  (for _array_element (Variables|Default|GetVisibleComponents)
    (Development|SetHiddeninGame _array_element true true)))

(fn OnRep_bLootAvailable
  (if (not (Variables|Default|GetbLootAvailable))
    (Collision|SetCollisionEnabled (Variables|Default|GetCollisionCapsule) "NoCollision")
    (Variables|Interaction|SetbAllowInteract false)
    (MarkableActor|SetCanBeMarked false)
    (Networking|SwitchHasAuthority
      (:Authority
        (Gameplay|Die "(TagName=\"WildEstate.BushHerb.Harvestable.ClaimedDestroy\")" (Pawn|GetController (Variables|Default|GetLastInteractingPawn)) (Variables|Default|GetLastInteractingPawn))))))

(fn PawnHasOpenInventorySlot
  (params (PawnToCheck :type Pawn%20Object%20Reference))
  (bind _asfort_player_controller (Utilities|Casting|CastToFortPlayerController (Pawn|GetController PawnToCheck))
    (:then
      (bind _returnvalue (IndexSelectionModel|GetNumItems(Message) _asfort_player_controller))
      (return (< _returnvalue 5)))
    (:CastFailed
      (return false))))

(fn PrepDynamicMaterialInstance
  (for (_array_element _array_index) (Variables|Default|GetVisibleComponents)
    (bind _returnvalue (Rendering|Material|CreateDynamicMaterialInstance (Utilities|Casting|CastToStaticMeshComponent _array_element) 1 0 "None" :!class /Script/Engine.PrimitiveComponent))
    (Utilities|Array|AddUnique (Variables|Default|GetDMI_VisComp_MatInstance) _returnvalue) ; Store each components material instance
    (bind _returnvalue_2 (Rendering|Material|GetScalarParameterValue _returnvalue "SpriteScale" :!class /Script/Engine.MaterialInstanceDynamic))
    (Utilities|Array|SetArrayElem (Variables|Default|GetVisComp_DefaultSpriteScale) _array_index _returnvalue_2 true) ; Store default scale
  ))

(fn ConstructionScript
  (CallFunction|CacheVisualComponents))
