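; Harvestable tomato bush variant (TomatoC) in the BushHerb consumables set, built on a parent Blueprint that supplies its behavior.
; Its CacheVisualComponents override calls the parent, then appends six stack components to the VisibleComponents array.
; Those components are named AppleStack1 through AppleStack6 despite the tomato asset name.
; The BeginPlay, overlap and Tick events are unused stubs, and the construction script only calls the parent.
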(event EventBeginPlay
  (|CallParentFunction) ; This node is disabled and will not be called.\nDrag off pins to build functionality.
)

(event Collision|EventActorBeginOverlap
  (params (OtherActor :type Actor%20Object%20Reference)))

(event EventTick
  (params (DeltaSeconds :type Float%20%28single-precision%29)))

(fn CacheVisualComponents
  (|CallParentFunction)
  (Utilities|Array|AppendArray (Variables|Default|GetVisibleComponents) (Utilities|Array|MakeArray (Variables|Default|GetAppleStack1) (Variables|Default|GetAppleStack2) (Variables|Default|GetAppleStack3) (Variables|Default|GetAppleStack4) (Variables|Default|GetAppleStack5) (Variables|Default|GetAppleStack6))))

(fn ConstructionScript
  (|CallParentFunction))
