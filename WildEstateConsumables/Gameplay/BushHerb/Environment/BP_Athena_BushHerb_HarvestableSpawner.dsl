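; Harvestable spawner for BushHerb that spawns a tomato or enhanced variant building actor, then silently destroys itself.
; On authority, once playlist data is ready, scalable-float rows gate spawning by enabled state and chance, with a force override.
; Unauthored variants are rolled by container placement, with optional random yaw, then resolved through variant-to-class maps.
; In editor, construction previews the selected class by copying its static meshes with a yellow material.
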(event EventBeginPlay
  (bind _row_spawnenhancedtypebool (Variables|Default|GetROW_SpawnEnhancedTypeBool))
  (bind _returnvalue (ScalableFloat|GetValueasBool _row_spawnenhancedtypebool))
  (bind _stackvarianttospawn (Variables|BushHerb|GetStackVariantToSpawn))
  (bind _returnvalue_2 (Math|Conversions|ToInteger(Byte) _stackvarianttospawn))
  (bind _map_selectedtypetoenhancedvariantselected (Variables|Default|GetMAP_SelectedTypeToEnhancedVariantSelected))
  (bind _map_selectedtypetotomatovariantselected (Variables|Default|GetMAP_SelectedTypeToTomatoVariantSelected))
  (bind _returnvalue_3 (select _returnvalue _map_selectedtypetoenhancedvariantselected _map_selectedtypetotomatovariantselected))
  (bind (_value _returnvalue_4) (Utilities|Map|Find _returnvalue_3 _returnvalue_2))
  (|CallParentFunction)
  (Networking|SwitchHasAuthority
    (:Authority
      (Class|FortAsyncActionCurrentPlaylistReady|PlaylistDataReadyAsync
        (:OnReady
          (Utilities|FlowControl|Sequence
            (:then_0
              ;[0.2087 0.3689 0.7031] Check if this spawner can spawn the actor
              (if (and (ScalableFloat|GetValueasBool (Variables|Default|GetROW_SpawningEnabled)) (or (<= (ScalableFloat|GetValueatLevel (Variables|Default|GetROW_SpawnChancePct)) (Math|Random|RandomFloat)) (Variables|BushHerb|GetbForceSpawnEnabled)))
                ;[0.02245 0.03488 1] Spawn the selected variant and type - or randomize if it not authored
                (switch Utilities|FlowControl|Switch|SwitchonE_BushHerb_SpawnerVariant _stackvarianttospawn
                  (:NewEnumerator0
                    (BushHerb|RollRandomSpawn :!label _label) ; if a variant isn't already selected, get a valid one
                  )
                  (:NewEnumerator1
                    _)
                  (:NewEnumerator2
                    _)
                  (:NewEnumerator3
                    _)
                  (:NewEnumerator4
                    (goto _label)))
                (if _returnvalue_4
                  (Building|SpawnBuildingGameplayActor _value (Transformation|GetActorTransform) self true))))
            (:then_1
              ;[1 0.3722 0] Destroy this spawner
              (Gameplay|SilentDie false) ; we always want the spawner to die after attempting to spawn
            )))))))

(fn RollRandomSpawn
  ;[0.2087 0.3689 0.7031] Randomly rotate if that is allowed
  (if (Variables|GetbRandomYawOnPlacement)
    (Transformation|SetActorRotation (Math|Rotator|MakeRotator 0.0 0.0 (Math|Random|RandomFloatinRange 0.0 3.6e+02))))
  ;[0.2087 0.3689 0.7031] Select a valid variant
  (if (Variables|BushHerb|GetbIsPlacedInContainer)
    (Variables|BushHerb|SetStackVariantToSpawn (select (Math|Random|RandomBool) "NewEnumerator2" "NewEnumerator1"))
    (else
      (Variables|BushHerb|SetStackVariantToSpawn (select (Math|Random|RandomBool) "NewEnumerator3" "NewEnumerator2"))))
  ;[0.01306 0.02447 0.1146] Editor spawning tool: setup the visuals of the preview mesh and make the actor dirty for proper saving
  (if (Editor|InEditororPIE)
    (Debug|SetPreviewActorDebugVis)
    (Transactions|TransactObject self) ; makes asset dirty to save tag update
  ))

(fn SetPreviewActorDebugVis
  (bind _row_spawnenhancedtypebool (Variables|Default|GetROW_SpawnEnhancedTypeBool))
  (bind _returnvalue (ScalableFloat|GetValueasBool _row_spawnenhancedtypebool))
  (bind _stackvarianttospawn (Variables|BushHerb|GetStackVariantToSpawn))
  (bind _returnvalue_2 (Math|Conversions|ToInteger(Byte) _stackvarianttospawn))
  (bind _map_selectedtypetoenhancedvariantselected (Variables|Default|GetMAP_SelectedTypeToEnhancedVariantSelected))
  (bind _map_selectedtypetotomatovariantselected (Variables|Default|GetMAP_SelectedTypeToTomatoVariantSelected))
  (bind _returnvalue_3 (select _returnvalue _map_selectedtypetoenhancedvariantselected _map_selectedtypetotomatovariantselected))
  (bind (_value _returnvalue_4) (Utilities|Map|Find _returnvalue_3 _returnvalue_2))
  (if _returnvalue_4
    (for _array_element (Fortnite|GetStaticMeshComponentsforClass _value)
      (bind _returnvalue_5 (Game|AddComponentbyClass 0 "/Script/Engine.StaticMeshComponent" false (Transformation|GetRelativeTransform _array_element)))
      (Components|StaticMesh|SetStaticMesh _returnvalue_5 (Variables|/Script/Engine.StaticMeshComponent|GetStaticMesh _array_element))
      (Rendering|Material|SetMaterial _returnvalue_5 0 "/Game/Athena/DrivableVehicles/M_VehicleSpawner_Yellow.M_VehicleSpawner_Yellow"))
    (else
      (Rendering|SetVisibility (Variables|Default|GetPreviewMesh) true false))))

(fn ConstructionScript
  (if (Editor|InEditororPIE)
    (Debug|SetPreviewActorDebugVis)))
